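# Animation Blueprint for the minor spire creature in SpireBoss, playing a looping idle with additive body hit reactions on top.
# The idle state machine feeds DefaultSlot, then a linked AdditiveHitReacts layer applies the FullbodyAdditive slot and a HitReact state machine.
# SetHitReact_Body raises bHitReact for a retriggerable 0.25-second window.
# A global state alias enters a shuffled left/right reaction while it is set, returning to an additive identity passthrough when cleared.
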AnimBlueprint /SpireBoss/Assets/Animation/SpireMinor/Spires_Minor_AnimBlueprint
target_skeleton=/SpireBoss/GameplayActors/Meshes/Spires_Minor_Creature_Skeleton
graphs [
 AdditiveHitReacts {nodes [
   n1 = Use cached pose 'Input Pose'()
   AnimGraphNode_StateMachine_0 = HitReact() {is=title AnimGraphNode_StateMachine}
   Output Pose(Result=Apply Additive(Base=Slot 'FullbodyAdditive'(Source=@n1) {slotName=FullbodyAdditive}, Additive=@AnimGraphNode_StateMachine_0) {alphaBoolBlend {blendOption=Linear}})
   Save cached pose 'Input Pose'(Pose=Input Pose() {name=InPose_0}) {cacheName=Input Pose}
  ]}
 HitReact {
  entry=Passthrough
  path=AdditiveHitReacts.AnimGraphNode_StateMachine_0.HitReact
  ids=id
  nodes [
   To Hit React() {is=title AnimStateAliasNode, bGlobalAlias=true}
   AnimStateNode_0 = Passthrough() {is=title AnimStateNode}
   AnimStateNode_1 = HitReact_Body() {
    is=title AnimStateNode
    stateEntered {guid=#1}
    stateFullyBlended {guid=#2}
    stateLeft {guid=#3}
   }
   AnimStateTransitionNode_0 = AnimStateTransitionNode()
   AnimStateTransitionNode_2 = AnimStateTransitionNode() {
    crossfadeDuration=0.15
    transitionEnd {guid=#4}
    transitionInterrupt {guid=#5}
    transitionStart {guid=#6}
   }
  ]
  links [
   {
    dst=Passthrough
    rule=AdditiveHitReacts.AnimGraphNode_StateMachine_0.HitReact.AnimStateTransitionNode_0.Transition
    src=HitReact_Body
    via=AnimStateTransitionNode_0
   }
   {
    dst=HitReact_Body
    rule=AdditiveHitReacts.AnimGraphNode_StateMachine_0.HitReact.AnimStateTransitionNode_2.AnimationTransitionGraph_0
    src=To Hit React
    via=AnimStateTransitionNode_2
   }
  ]
 }
 Passthrough {
  path=AdditiveHitReacts.AnimGraphNode_StateMachine_0.HitReact.AnimStateNode_0.Passthrough
  nodes [Output Animation Pose(Result=Additive Identity Pose())]
 }
 HitReact_Body {
  path=AdditiveHitReacts.AnimGraphNode_StateMachine_0.HitReact.AnimStateNode_1.HitReact_Body
  nodes [
   n1 = Random Sequence Player!() -> (Pose:Pose Link Structure) {
    is=title AnimGraphNode_RandomPlayer
    bShuffleMode=true
    entries [
     {
      chanceToPlay=0.5
      maxLoopCount=0
      maxPlayRate=1
      minLoopCount=0
      minPlayRate=1
      sequence=&/SpireMinor_Hitreact_L
      blendIn {blendOption=Linear, blendTime=0.2, customCurve=null}
     }
     {
      chanceToPlay=0.5
      maxLoopCount=0
      maxPlayRate=1
      minLoopCount=0
      minPlayRate=1
      sequence=&/SpireMinor_Hitreact_R
      blendIn {blendOption=Linear, blendTime=0.2, customCurve=null}
     }
    ]
   }
   Output Animation Pose(Result=@n1)
  ]
 }
 Transition {
  path=AdditiveHitReacts.AnimGraphNode_StateMachine_0.HitReact.AnimStateTransitionNode_0.Transition
  nodes [Result(bCanEnterTransition=NOT Boolean(A=Get bHitReact() -> (bHitReact:bool)))]
 }
 AnimationTransitionGraph_0 {
  path=AdditiveHitReacts.AnimGraphNode_StateMachine_0.HitReact.AnimStateTransitionNode_2.AnimationTransitionGraph_0
  nodes [Result(bCanEnterTransition=Get bHitReact())]
 }
 AnimGraph {nodes [
   Output Pose(Result=Use cached pose 'AdditiveLayer'())
   AnimGraphNode_StateMachine_2 = SpireMinor_Idle() {is=title AnimGraphNode_StateMachine}
   Save cached pose 'BasePose'(Pose=Slot 'DefaultSlot'(Source=@AnimGraphNode_StateMachine_2)) {cacheName=BasePose}
   Save cached pose 'AdditiveLayer'(Pose=AdditiveHitReacts!(InPose_0:Pose Link Structure=Use cached pose 'BasePose'()) -> (Pose:Pose Link Structure) {is=title AnimGraphNode_LinkedAnimLayer, layer=AdditiveHitReacts}) {cacheName=AdditiveLayer}
  ]}
 SpireMinor_Idle {entry=Idle, path=AnimGraph.AnimGraphNode_StateMachine_2.SpireMinor_Idle, ids=id, nodes [AnimStateNode_0 = Idle()]}
 Idle {
  path=AnimGraph.AnimGraphNode_StateMachine_2.SpireMinor_Idle.AnimStateNode_0.Idle
  nodes [Output Animation Pose(Result=Sequence Player 'SpireMinor_Idle'() {sequence=&/SpireMinor_Idle})]
 }
 EventGraph {nodes [
   SetHitReact_Body() {is=title K2Node_CustomEvent}
   > Set bHitReact(bHitReact:bool=true) -> (Output_Get:bool)
   > RetriggerableDelay(Duration=0.25)
   > Set bHitReact(bHitReact=false)
   TryGetPawnOwner()
   Event BlueprintUpdateAnimation()
  ]}
 OnBodyHitBecomeRelevant {nodes [
   OnBodyHitBecomeRelevant!() -> (then, Context:"Anim Update Context Structure (by ref)", Node:"Anim Node Reference Structure (by ref)") {is=title K2Node_FunctionEntry}
   > Set bHitReact(bHitReact=false)
   > Return Node()
  ]}
]
variables [bHitReact {type=boolean}]
